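# Parent material for the EctoCat backpack, built on the 2023 character parent functions with only the emissive output overridden.
# Local-position X/Z becomes radial coordinates for a panning cloud noise that modulates the emissive color, with a slow sine pulse on brightness.
# The noise sample takes its mip derivatives from the mesh UVs rather than the radial coordinates.
# A time-of-day scalar dims the glow at night.
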Material /Game/Accessories/FORT_Backpacks/Backpack_M_MED_EctoCat/Materials/M_EctoCat_Backpack_PARENT
graph {
 properties {
  bUseMaterialAttributes=true
  bUsedWithMeshDeformer=true
  bUsedWithNanite=true
  bUsedWithSkeletalMesh=true
  pixelDepthOffsetMode=PDOM_Legacy
  shadingModel=MSM_FromMaterialExpression
  usedShadingModels=Default Lit | Subsurface | Clear Coat | Subsurface Profile | Cloth
 }
 nodes [
  n1 = MaterialFunctionCall() {materialFunction=&MF_CharacterParent_2023_PreFX}
  n2 = GetMaterialAttributes(None=MaterialFunctionCall(Material=@n1.Result, M=@n1.Result, Pre Skinned Local Position=@n1.Result, Distance Blend=@n1.Result, Fresnel=@n1.Result) {materialFunction=&MF_CharacterParent_2023_PostFX}.Result) {attributeGetTypes [#1]}
  n3 = Constant() {r=1}
  n4 = Constant() {r=0.55}
  n5 = MaterialFunctionCall() {materialFunction=&LocalPosition}
  n6 = Constant() {r=0.07}
  n7 = Add(A=ComponentMask(None=Multiply(A=@n5.Local Position, B=@n6) {constB=0.01}) {b=true, r=true}, B=Multiply(A=Constant2Vector() {g=1.5}, B=@n6))
  n8 = TextureCoordinate()
  n9 = TextureSample!(
   UVs=Panner(Coordinate=MaterialFunctionCall(Vector or UVs=@n7) {materialFunction=&VectorToRadialValue}.Radial Coordinates) {speedX=0.04, speedY=-0.08}
   Tex
   "DDX(UVs)"=DDX(Value=@n8)
   "DDY(UVs)"=DDY(Value=@n8)
   Apply View MipBias
  ) {mipValueMode=TMVM_Derivative, texture=&AtmosphericCloudNoise01}
  n10 = Multiply(A=@n2.Emissive Color, B=Saturate(None=Add(A=@n9.RGB) {constB=0.275}))
  n11 = Sine(None=Multiply(A=Time()) {constB=0.4})
  n12 = LinearInterpolate(Alpha=Add(A=Multiply(A=@n11) {constB=0.5}) {constB=0.5}) {constA=1.2, constB=2.15}
  SetMaterialAttributes(
   MaterialAttributes=@n2.MaterialAttributes
   Emissive Color=Multiply(
    A=MaterialFunctionCall(Value_Morning=@n3, Value_Day=@n3, Value_Evening=@n3, Value_Night=@n4) {materialFunction=&Scalar_Time_of_Day}.Result
    B=Multiply(A=@n10, B=@n12)
   )
  ) {attributeSetTypes [#1]}
 ]
}
